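# Material function that gives the summer Punk Koi character body a view-angle-dependent jade tint.
# The clamped dot product of camera vector and pixel normal indexes curve atlas rows for a color gradient and an opacity value.
# The gradient is brightened and multiplied into the input color, while a separately tinted, scaled copy feeds the SSColor output.
# A channel mask on the second curve supplies the Opacity output.
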MaterialFunction /BRCosmetics/Characters/Player/Female/Medium/Bodies/F_MED_Punk_Koi_Summer/Materials/MF_JadeTint
graph {nodes [
  n1 = Clamp(None=DotProduct(A=CameraVectorWS(), B=PixelNormalWS()))
  n2 = CurveAtlasRowParameter(CurveTime=@n1) {
   atlas=/BRCosmetics/Characters/Player/Female/Medium/Bodies/F_MED_Punk_Koi_Summer/Texture/PunkKoiAtlas
   curve=/BRCosmetics/Characters/Player/Female/Medium/Bodies/F_MED_Punk_Koi_Summer/Texture/PunkKoiJadeCurve
   parameterName=Color
  }
  n3 = ScalarParameter() {defaultValue=4, parameterName=ColorMult}
  FunctionOutput(None=Multiply(A=FunctionInput() {inputName=Color}, B=Multiply(A=@n2, B=@n3)))
  n4 = CurveAtlasRowParameter(CurveTime=@n1) {
   atlas=&PunkKoiAtlas
   curve=/BRCosmetics/Characters/Player/Female/Medium/Bodies/F_MED_Punk_Koi_Summer/Texture/PunkKoisSSCurve
   defaultValue=1
   parameterName=Opacity
  }
  FunctionOutput(None=ChannelMaskParameter(None=@n4) {parameterName=OpacityChannel}.R) {outputName=Opacity}
  n5 = VectorParameter() {parameterName=SS_COlorMult, defaultValue=<rgba 0 1 0.6901 1>}
  FunctionOutput(None=Multiply(A=Multiply(A=@n2, B=@n5.RGB), B=ScalarParameter() {defaultValue=0.5, parameterName=SS_ColorMult})) {outputName=SSColor}
 ]}
