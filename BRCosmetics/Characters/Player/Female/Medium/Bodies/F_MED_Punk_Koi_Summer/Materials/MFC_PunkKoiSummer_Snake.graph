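# Material function layering an animated snake with wispy emissive smoke onto the Punk Koi Summer character's material.
# A pre-skinned position mask flips UVs per side, which then scroll to form snake UVs.
# Flow-map-warped cloud noise, crossfaded in two phases, carves the smoke mask tinting diffuse and emissive.
# Static switches select jade tinting and enable the layer, which passes through MF_MetalsLUT and blends over the input attributes.
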MaterialFunction /BRCosmetics/Characters/Player/Female/Medium/Bodies/F_MED_Punk_Koi_Summer/Materials/MFC_PunkKoiSummer_Snake
graph {nodes [
  n1 = NamedRerouteUsage()
  n2 = Multiply(
   A=TextureCoordinate()
   B=AppendVector(A=ScalarParameter() {defaultValue=5, parameterName=FlowMapTiling_X}, B=ScalarParameter() {defaultValue=0.25, parameterName=FlowMapTiling_Y})
  )
  n3 = TextureSample(UVs=LinearInterpolate(A=Panner(Coordinate=@n2) {speedX=-0.2, speedY=0.1}, B=Panner(Coordinate=@n2) {speedX=0.2, speedY=0.1}, Alpha=NamedRerouteUsage())) {
   samplerType=SAMPLERTYPE_LinearColor
   texture=/Game/Creative/Sets/PropSets/Primitives/Customization/Textures/GalacticSwirl/T_CM_GalacticSwirl_Flowmap
  }
  n4 = Subtract(A=Multiply(A=ComponentMask(None=@n3.RGB) {g=true, r=true}) {constB=2})
  Add(A=Multiply(A=ComponentMask(None=@n1) {g=true, r=true}, B=ScalarParameter() {defaultValue=0.25, parameterName=SnakeNormalInfluence}), B=@n4)
  AppendVector(A=ScalarParameter() {defaultValue=0.132, parameterName=PAss}, B=Constant())
  Constant() {r=1}
  n5 = TextureSample(UVs=NamedRerouteUsage()) {
   samplerType=SAMPLERTYPE_Normal
   texture=/BRCosmetics/Characters/Player/Female/Medium/Bodies/F_MED_Punk_Koi_Summer/Texture/T_F_MED_Punk_Koi_Summer_VFX_N
  }
  Fresnel(ExponentIn=ScalarParameter() {parameterName=FresnelExponent}, Normal=Normalize(VectorInput=Transform(None=@n5.RGB)))
  n6 = TextureSample(UVs=NamedRerouteUsage()) {texture=/BRCosmetics/Characters/Player/Female/Medium/Bodies/F_MED_Punk_Koi_Summer/Texture/T_F_MED_Punk_Koi_Summer_VFX_D}
  n7 = NamedRerouteUsage()
  n8 = Multiply(
   A=TextureCoordinate()
   B=AppendVector(A=ScalarParameter() {defaultValue=4, parameterName=Noise_Tiling_X}, B=ScalarParameter() {defaultValue=0.25, parameterName=Noise_Tiling_Y})
  )
  n9 = ScalarParameter() {defaultValue=0.25, parameterName=" WarpIntewnsity"}
  n10 = Multiply(A=Time(), B=ScalarParameter() {defaultValue=0.025, parameterName=Speed})
  n11 = Frac(None=@n10)
  n12 = Add(A=@n8, B=Multiply(A=NamedRerouteUsage(), B=Multiply(A=@n9, B=@n11)))
  n13 = Frac(None=Add(A=@n10) {constB=0.5})
  n14 = Add(A=@n8, B=Multiply(A=NamedRerouteUsage(), B=Multiply(A=@n9, B=@n13)))
  n15 = Subtract(A=Multiply(A=@n11) {constB=2})
  n16 = Multiply(
   A=LinearInterpolate(
    A=TextureSample(UVs=@n12) {samplerType=SAMPLERTYPE_LinearGrayscale, texture=/Game/Athena/Prototype/Terrain/T_CloudCoverage}.RGB
    B=TextureSample(UVs=@n14) {samplerType=SAMPLERTYPE_LinearGrayscale, texture=&T_CloudCoverage}.RGB
    Alpha=Abs(None=@n15)
   )
   B=ScalarParameter() {defaultValue=0.5, parameterName=NoiseStrength}
  )
  n17 = Clamp(None=Subtract(A=@n7, B=@n16))
  n18 = Multiply(
   A=Multiply(A=NamedRerouteDeclaration(None=@n17) {name=WhispySmokeMask, nodeColor=<rgba 0.4766 1 0 1>})
   B=VectorParameter() {parameterName=EmissiveColorSnake, defaultValue=<rgba 0 1 0.515 1>}.RGB
  )
  n19 = NamedRerouteUsage()
  n20 = ScalarParameter() {defaultValue=1, parameterName=SmokeBlendAmt}
  n21 = LinearInterpolate(A=@n6.RGB, B=Multiply(A=@n18, B=ScalarParameter() {defaultValue=0.5, parameterName=DiffuseMult}), Alpha=Clamp(None=Multiply(A=@n19, B=@n20)))
  n22 = MaterialFunctionCall!(Color=@n21) {materialFunction=&/MF_JadeTint}
  n23 = TextureSample(UVs=NamedRerouteUsage()) {
   samplerType=SAMPLERTYPE_Masks
   texture=/BRCosmetics/Characters/Player/Female/Medium/Bodies/F_MED_Punk_Koi_Summer/Texture/T_F_MED_Punk_Koi_Summer_VFX_M
  }
  n24 = ChannelMaskParameter(None=@n23.RGB) {maskChannel=Green, parameterName=_M_Channel, defaultValue=<rgba 0 1 0 0>}
  n25 = ScalarParameter() {defaultValue=1.5, parameterName=AO_Mult}
  n26 = Power(Base=Clamp(None=Multiply(A=@n24.R, B=@n25)), Exp=ScalarParameter() {defaultValue=4, parameterName=AO_Pow}) {constExponent=4}
  n27 = TextureSample(UVs=NamedRerouteUsage()) {
   samplerType=SAMPLERTYPE_Masks
   texture=/BRCosmetics/Characters/Player/Female/Medium/Bodies/F_MED_Punk_Koi_Summer/Texture/T_F_MED_Punk_Koi_Summer_VFX_S
  }
  n28 = Power(
   Base=Multiply(A=@n18, B=ScalarParameter() {defaultValue=0.1, parameterName=EmissiveStrengthSnake})
   Exp=ScalarParameter() {defaultValue=1, parameterName=FresnelPowerSnake}
  )
  n29 = GetMaterialAttributes(None=FunctionInput() {inputName=Material, inputType=FunctionInput_MaterialAttributes}) {attributeGetTypes [#1, #2]}
  n30 = SetMaterialAttributes!(
   MaterialAttributes=@n29.MaterialAttributes
   BaseColor=Multiply(A=@n22.Result, B=@n26)
   Metallic=Constant()
   Specular=@n27.R
   Roughness=@n27.R
   Normal=@n5.RGB
   EmissiveColor=@n28
   ShadingModel=ShadingModel()
   Opacity
   SubsurfaceColor
  ) {attributeSetTypes [
    #3
    #4
    #5
    #6
    #2
    #1
    #7
    #8
    #9
   ]}
  n31 = SetMaterialAttributes!(
   MaterialAttributes=@n29.MaterialAttributes
   BaseColor=Multiply(A=@n21, B=@n26)
   Metallic=@n27.G
   Specular=@n27.G
   Roughness=@n27.G
   Normal=@n5.RGB
   EmissiveColor=@n28
  ) {attributeSetTypes [
    #3
    #4
    #5
    #6
    #2
    #1
   ]}
  n32 = StaticSwitchParameter(True=@n30, False=@n31) {parameterName=UseJadeColorRamp}
  n33 = MaterialFunctionCall(Mask=Constant() {r=1}, Materials=@n32) {materialFunction=&MF_MetalsLUT}
  n34 = NamedRerouteUsage()
  FunctionOutput(None=StaticSwitchParameter(True=BlendMaterialAttributes(A=@n33.Result, B=@n32, Alpha=@n34), False=@n29.MaterialAttributes) {parameterName=UsePunkKoiSnake})
  Multiply(A=@n6.RGB)
  Multiply(A=OneMinus(None=@n23.R))
  n35 = MaterialFunctionCall(Float2=TextureCoordinate()) {materialFunction=&BreakOutFloat2Components}
  n36 = PreSkinnedPosition()
  n37 = ShaderStageSwitch(PixelShader=VertexInterpolator(VS=@n36).PS, VertexShader=@n36)
  n38 = ChannelMaskParameter(None=Divide(A=FunctionInput(Preview=@n37) {bUsePreviewValueAsDefault=true, inputName=PreskinnedPos}, B=PreSkinnedLocalBounds().Max))
  n39 = ScalarParameter() {defaultValue=-0.2, parameterName=Offset, sliderMin=1}
  n40 = Clamp(None=Ceil(None=Subtract(A=@n38.R, B=@n39)))
  n41 = MaterialFunctionCall(X=LinearInterpolate(A=OneMinus(None=@n35.R), B=@n35.R, Alpha=@n40), Y=LinearInterpolate(A=OneMinus(None=@n35.G), B=@n35.G, Alpha=@n40)) {materialFunction=&MakeFloat2}
  n42 = AppendVector(A=ScalarParameter() {defaultValue=1, parameterName=SnakeTiling}, B=Constant() {r=1})
  n43 = Time()
  n44 = AppendVector(A=ScalarParameter() {defaultValue=0.025, parameterName=Speed}, B=Constant())
  NamedRerouteDeclaration(None=Add(A=Multiply(A=@n41.Result, B=@n42), B=Multiply(A=@n43, B=@n44))) {name=SnakeUV, nodeColor=<rgba 1 0 0.0937 1>}
  NamedRerouteDeclaration(None=@n35.R) {name=XGradient, nodeColor=<rgba 0 1 0.3438 1>}
  NamedRerouteDeclaration(None=Subtract(A=Multiply(A=@n5.RGB) {constB=2})) {name=SnakeNormal, nodeColor=<rgba 0.4531 0 1 1>}
  n45 = Constant() {r=0.5}
  n46 = ScalarParameter() {defaultValue=0.34, parameterName=GradientRTadius}
  n47 = ScalarParameter() {parameterName=GradientHardness}
  NamedRerouteDeclaration(None=OneMinus(None=SphereMask(A=@n35.G, B=@n45, Radius=@n46, Hardness=@n47))) {name=Y_Gradient, nodeColor=<rgba 1 0.2812 0 1>}
  NamedRerouteDeclaration(None=@n4) {name=FLowMapTexture, nodeColor=<rgba 0.9922 0 1 1>}
  NamedRerouteDeclaration(None=@n40) {name=ArmMask, nodeColor=<rgba 0.3828 0 1 1>}
  NamedRerouteUsage()
  NamedRerouteUsage()
  n48 = NamedRerouteUsage()
  n49 = Time()
  n50 = ScalarParameter() {defaultValue=0.1, parameterName=EmissiveSpeed}
  n51 = Add(A=Add(A=ComponentMask(None=@n48) {r=true}, B=Multiply(A=@n49, B=@n50)), B=ScalarParameter() {defaultValue=0.5, parameterName=Offset}) {constB=-0.25}
  OneMinus(None=Frac(None=@n51))
  Power(Base=@n7)
  Reroute(None=Multiply(A=@n22.Opacity, B=OneMinus(None=@n26)))
  Reroute(None=Multiply(A=@n22.SSColor, B=@n26))
  ScalarParameter() {parameterName=Noise}
  n52 = NamedRerouteUsage()
  SetMaterialAttributes!(
   MaterialAttributes=@n29.MaterialAttributes
   BaseColor=@n52
   Metallic
   Specular
   Roughness
   Normal
   EmissiveColor=@n52
   ShadingModel
   Opacity
   SubsurfaceColor
  ) {attributeSetTypes [
    #3
    #4
    #5
    #6
    #2
    #1
    #7
    #8
    #9
   ]}
  Sine()
  Subtract(A=ConstantBiasScale(None=Sine(None=@n51)), B=@n16)
 ]}
