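# Master character material for the Punk Koi Summer female outfit body, layering skin and snake effects onto the shared character shading.
# It chains the base PreFX function through CharacterSkinFX, an outfit-specific Snake function and the base PostFX function.
# PreFX also supplies pre-skinned local position to the effect stages.
# It uses material attributes, with shading models chosen by expression and a hero skin subsurface profile.
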Material /BRCosmetics/Characters/Player/Female/Medium/Bodies/F_MED_Punk_Koi_Summer/Materials/M_Punk_Koi_Summer_MASTER
graph {
 properties {
  bUseMaterialAttributes=true
  bUsedWithMeshDeformer=true
  bUsedWithMorphTargets=true
  bUsedWithSkeletalMesh=true
  pixelDepthOffsetMode=PDOM_Legacy
  refractionMethod=RM_IndexOfRefraction
  shadingModel=MSM_FromMaterialExpression
  subsurfaceProfile=&SS_HeroSkin_02
  usedShadingModels=Default Lit | Subsurface | Clear Coat | Subsurface Profile | Cloth | Eye
 }
 nodes [
  n1 = MaterialFunctionCall() {materialFunction=&BaseFN_CharacterFunction_PreFX}
  MaterialFunctionCall(
   AO=@n1.AO
   Distance Blend=@n1.AO
   Fresnel=@n1.AO
   Gmap=@n1.AO
   Pre Skinned Local Position=@n1.Pre skinned Local Position
   Material=MaterialFunctionCall!(
    PreskinnedPos=@n1.Pre skinned Local Position
    Material=MaterialFunctionCall(
     PreskinnedPosition=@n1.Pre skinned Local Position
     Gmap=@n1.Pre skinned Local Position
     MaterialAttributes=@n1.Pre skinned Local Position
     AO=@n1.Pre skinned Local Position
    ) {materialFunction=&CharacterSkinFX}.Result
   ) {materialFunction=&/MFC_PunkKoiSummer_Snake}.Result
  ) {materialFunction=&BaseFN_CharacterFunction_PostFX}
 ]
}
