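# Cosmetic AnimBP for the female Punk Koi Summer body, adding secondary motion to its belt, belt snake and thigh armor.
# The parent mesh's pose runs through a deform Control Rig, constraints and additive bone corrections into a cached pre-sim pose.
# Thigh and spine rotation deltas, plus crouch-walk state, drive correction alphas.
# AnimDynamics then simulates two snake setups blended by crouching, followed by thigh armor, disabled while sliding.
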AnimBlueprint /BRCosmetics/Characters/Player/Female/Medium/Bodies/F_MED_Punk_Koi_Summer/Meshes/F_MED_Punk_Koi_Summer_AnimBP
target_skeleton=&/F_MED_Punk_Koi_Summer_Skeleton
graphs [
 AnimGraph {nodes [
   "Constraint - dyn_thigh_armor (P)"(Alpha=1, ConstraintWeights_0=1) {
    alphaBoolBlend {blendOption=Linear}
    boneToModify {boneName=dyn_thigh_armor}
    constraintSetup [{targetBone {boneName=pelvis}}]
    constraintWeights [1]
   }
   n1 = "Transform (Modify) Bone - Bone: dyn_belt_snake__MSTR"!(
    ComponentPose=Local To Component(LocalPose=Use cached pose 'presim'())
    bAlphaBoolEnabled:bool=Get bIsMoving2D()
    Translation="0,0,0"
    Rotation="35,-10,-15"
    Scale="1,1,1"
   ) -> (Pose) {
    alphaInputType=Bool
    rotationMode=BMM_Additive
    rotationSpace=BCS_BoneSpace
    translationMode=BMM_Additive
    translationSpace=BCS_BoneSpace
    alphaBoolBlend {blendInTime=0.1, blendOption=Linear, blendOutTime=0.1}
    boneToModify {boneName=dyn_belt_snake__MSTR}
    rotation=<rot 35 -10 -15>
   }
   n2 = Anim Dynamics - Bone: dyn_belt_snake_01!(ComponentPose=@n1, bAlphaBoolEnabled:bool=True, ExternalForce="0,0,0") -> (Pose) {
    alphaInputType=Bool
    angularDampingOverride=0.997
    angularSpringConstant=50
    bAngularSpring=true
    bGravityOverrideInSimSpace=true
    bOverrideAngularDamping=true
    bUseGravityOverride=true
    simulationSpace=RootRelative
    alphaBoolBlend {blendInTime=0.1, blendOption=Linear, blendOutTime=0.1}
    boundBone {boneName=dyn_belt_snake_01}
    componentAppliedLinearAccClamp=<500 500 500>
    componentLinearAccScale=<0.3 0.3 0.3>
    componentLinearVelScale=<0.3 0.3 0.3>
    gravityOverride {z=-980}
    physicsBodyDefinitions [{
      boundBone {boneName=dyn_belt_snake_01}
      boxExtents=<30 20 20>
      constraintSetup {angularLimitsMax=<3 3 10>, angularLimitsMin=<-3 -45 -5>, angularTarget=<1 0 0>}
      localJointOffset=<1 0 0>
     }]
   }
   n3 = Anim Dynamics - Bone: dyn_belt_snake_02!(ComponentPose=@n2, bAlphaBoolEnabled:bool=True, ExternalForce="0,0,0") -> (Pose) {
    alphaInputType=Bool
    angularDampingOverride=0.997
    angularSpringConstant=100
    bAngularSpring=true
    bGravityOverrideInSimSpace=true
    bOverrideAngularDamping=true
    bUseGravityOverride=true
    simulationSpace=RootRelative
    alphaBoolBlend {blendInTime=0.1, blendOption=Linear, blendOutTime=0.1}
    boundBone {boneName=dyn_belt_snake_02}
    componentAppliedLinearAccClamp=<500 500 500>
    componentLinearAccScale=<0.3 0.3 0.3>
    componentLinearVelScale=<0.3 0.3 0.3>
    gravityOverride {z=-980}
    physicsBodyDefinitions [{
      collisionType=CustomSphere
      sphereCollisionRadius=2
      boundBone {boneName=dyn_belt_snake_02}
      boxExtents=<25 15 15>
      constraintSetup {angularLimitsMax=<5 0 10>, angularLimitsMin=<-1 -45 -5>, angularTarget=<1 0 0>}
      localJointOffset=<2 0 0>
     }]
    sphericalLimits [{limitRadius=9.5, drivingBone {boneName=thigh_twist_01_l}, sphereLocalOffset=<16 -3 1>}]
   }
   n4 = Component To Local(ComponentPose=@n3)
   n5 = "Transform (Modify) Bone - Bone: dyn_belt_snake__MSTR"(
    ComponentPose=Local To Component(LocalPose=Use cached pose 'presim'())
    bAlphaBoolEnabled=Get bIsJumping()
    Translation="0,0,0"
    Rotation="35,0,0"
    Scale="1,1,1"
   ) {
    alphaInputType=Bool
    rotationMode=BMM_Additive
    rotationSpace=BCS_BoneSpace
    translationMode=BMM_Additive
    translationSpace=BCS_BoneSpace
    alphaBoolBlend {blendInTime=0.1, blendOption=Linear, blendOutTime=0.1}
    boneToModify {boneName=dyn_belt_snake__MSTR}
    rotation {pitch=35}
   }
   n6 = Anim Dynamics - Bone: dyn_belt_snake_01!(ComponentPose=@n5, Alpha:float=1, ExternalForce="0,0,0") -> (Pose) {
    angularDampingOverride=0.97
    angularSpringConstant=50
    bGravityOverrideInSimSpace=true
    bOverrideAngularDamping=true
    bUseGravityOverride=true
    simulationSpace=RootRelative
    alphaBoolBlend {blendInTime=0.1, blendOption=Linear, blendOutTime=0.1}
    boundBone {boneName=dyn_belt_snake_01}
    componentAppliedLinearAccClamp=<500 500 500>
    componentLinearAccScale=<0.3 0.3 0.3>
    componentLinearVelScale=<0.3 0.3 0.3>
    gravityOverride {z=-980}
    physicsBodyDefinitions [{
      boundBone {boneName=dyn_belt_snake_01}
      boxExtents=<15 10 10>
      constraintSetup {angularLimitsMax=<3 3 25>, angularLimitsMin=<-3 -45 -5>, angularTarget=<1 0 0>}
      localJointOffset=<1 0 0>
     }]
   }
   n7 = Anim Dynamics - Bone: dyn_belt_snake_02!(ComponentPose=@n6, Alpha:float=1, ExternalForce="0,0,0") -> (Pose) {
    angularDampingOverride=0.97
    angularSpringConstant=50
    bGravityOverrideInSimSpace=true
    bOverrideAngularDamping=true
    bUseGravityOverride=true
    bUseSphericalLimits=true
    simulationSpace=RootRelative
    alphaBoolBlend {blendInTime=0.1, blendOption=Linear, blendOutTime=0.1}
    boundBone {boneName=dyn_belt_snake_02}
    componentAppliedLinearAccClamp=<500 500 500>
    componentLinearAccScale=<0.3 0.3 0.3>
    componentLinearVelScale=<0.3 0.3 0.3>
    gravityOverride {z=-980}
    physicsBodyDefinitions [{
      collisionType=CustomSphere
      sphereCollisionRadius=2
      boundBone {boneName=dyn_belt_snake_02}
      boxExtents=<15 10 10>
      constraintSetup {angularLimitsMax=<5 0 25>, angularLimitsMin=<-1 -45 -5>, angularTarget=<1 0 0>}
      localJointOffset=<2 0 0>
     }]
    sphericalLimits [{limitRadius=9.5, drivingBone {boneName=thigh_twist_01_l}, sphereLocalOffset=<16 -3 1>}]
   }
   n8 = Component To Local(ComponentPose=@n7)
   n9 = Get bIsCrouching()
   n10 = Anim Dynamics - Bone: dyn_thigh_armor!(
    ComponentPose=Local To Component(LocalPose=Blend Poses by bool(BlendPose_0=@n4, BlendPose_1=@n8, bActiveValue=@n9) {blendType=Linear})
    bAlphaBoolEnabled:bool=NOT Boolean(A=Get bIsSliding())
    ExternalForce="0,0,0"
   ) -> (Pose) {
    alphaInputType=Bool
    angularDampingOverride=0.97
    angularSpringConstant=50
    bAngularSpring=true
    bGravityOverrideInSimSpace=true
    bOverrideAngularDamping=true
    bUseGravityOverride=true
    simulationSpace=RootRelative
    alphaBoolBlend {blendInTime=0.1, blendOption=Linear, blendOutTime=0.1}
    boundBone {boneName=dyn_thigh_armor}
    componentAppliedLinearAccClamp=<2000 2000 2000>
    componentLinearAccScale=<0.3 0.3 0.4>
    componentLinearVelScale=<0.3 0.3 0.4>
    gravityOverride {z=-980}
    physicsBodyDefinitions [{
      boundBone {boneName=dyn_thigh_armor}
      boxExtents=<15 15 10>
      constraintSetup {angularLimitsMax=<1 0 3>, angularLimitsMin=<-1 -25 -3>, angularTarget=<1 0 0>}
      localJointOffset=<1 0 0>
     }]
   }
   Output Pose(Result=Component To Local(ComponentPose=@n10)) {layerGroup=None}
   n11 = Control Rig(Source=Copy Pose From Mesh() {bUseAttachedParent=true}) {
    alphaBoolBlend {blendOption=Linear}
    controlRigAssetReference {blueprintRigClass=&/F_MED_DeformRig_Punk_Koi_Summer_ControlRig.F_MED_DeformRig_Punk_Koi_Summer_ControlRig_C}
   }
   n12 = "Transform (Modify) Bone - Bone: dyn_thigh_armor"(
    ComponentPose=Local To Component(LocalPose=@n11)
    Alpha=Get Crouch Walk() -> (Crouch Walk:double)
    Translation="0,0,0.5"
    Rotation="0,0,-4"
    Scale="1,1,1"
   ) {
    rotationMode=BMM_Additive
    rotationSpace=BCS_BoneSpace
    translationMode=BMM_Additive
    translationSpace=BCS_BoneSpace
    alphaBoolBlend {blendOption=Linear}
    alphaScaleBiasClamp {bInterpResult=true}
    boneToModify {boneName=dyn_thigh_armor}
    rotation {roll=-4}
    translation {z=0.5}
   }
   n13 = "Constraint - dyn_belt_MSTR (PPPPPP)"!(
    ComponentPose=@n12
    Alpha=1
    ConstraintWeights_0=0.0039
    ConstraintWeights_1:float=0
    ConstraintWeights_2:float=0.0841
    ConstraintWeights_3:float=0.8638
    ConstraintWeights_4:float=0.011
    ConstraintWeights_5:float=0.0372
   ) -> (Pose) {
    alphaBoolBlend {blendOption=Linear}
    boneToModify {boneName=dyn_belt_MSTR}
    constraintSetup [
     {targetBone {boneName=spine_04}}
     {targetBone {boneName=spine_03}}
     {targetBone {boneName=spine_02}}
     {targetBone {boneName=pelvis}}
     {targetBone {boneName=thigh_twist_01_l}}
     {targetBone {boneName=deform_groin_l}}
    ]
    constraintWeights [1,1,1,1,1,1]
   }
   n14 = "Transform (Modify) Bone - Bone: dyn_belt"(ComponentPose=@n13, Alpha=Get ThighRotL_0() -> (ThighRotL_0:double), Translation="-7,0.000,1", Rotation="-10,-20,-5", Scale="1,1,1") {
    rotationMode=BMM_Additive
    rotationSpace=BCS_BoneSpace
    translationMode=BMM_Additive
    translationSpace=BCS_BoneSpace
    alphaBoolBlend {blendOption=Linear}
    boneToModify {boneName=dyn_belt}
    rotation=<rot -10 -20 -5>
    translation {x=-7, z=1}
   }
   n15 = "Constraint - dyn_belt_snake__MSTR (PPP)"!(ComponentPose=@n14, Alpha=1, ConstraintWeights_0=0.0181, ConstraintWeights_1:float=0.4819, ConstraintWeights_2:float=0.5) -> (Pose) {
    alphaBoolBlend {blendOption=Linear}
    boneToModify {boneName=dyn_belt_snake__MSTR}
    constraintSetup [{targetBone {boneName=spine_02}}, {targetBone {boneName=pelvis}}, {targetBone {boneName=dyn_belt}}]
    constraintWeights [1,1,1]
   }
   n16 = "Constraint - dyn_belt_snake__MSTR (R)"(ComponentPose=@n15, Alpha=1, ConstraintWeights_0=1) {
    alphaBoolBlend {blendOption=Linear}
    boneToModify {boneName=dyn_belt_snake__MSTR}
    constraintSetup [{transformType=Rotation, targetBone {boneName=thigh_l}}]
    constraintWeights [1]
   }
   n17 = "Transform (Modify) Bone - Bone: dyn_belt_snake__MSTR"!(ComponentPose=@n16, Alpha:float=Get ThighRotL_0(), Translation="0,0,0", Rotation="-40,35,25", Scale="1,1,1") -> (Pose) {
    rotationMode=BMM_Additive
    rotationSpace=BCS_BoneSpace
    translationMode=BMM_Additive
    translationSpace=BCS_BoneSpace
    alphaBoolBlend {blendOption=Linear}
    boneToModify {boneName=dyn_belt_snake__MSTR}
    rotation=<rot -40 35 25>
   }
   n18 = "Transform (Modify) Bone - Bone: dyn_belt_snake__MSTR"(ComponentPose=@n17, Alpha=Get Frontend Pose() -> (Frontend Pose:double), Translation="0,0,0", Rotation="10,0,0", Scale="1,1,1") {
    rotationMode=BMM_Additive
    rotationSpace=BCS_BoneSpace
    translationMode=BMM_Additive
    translationSpace=BCS_BoneSpace
    alphaBoolBlend {blendInTime=0.1, blendOption=Linear, blendOutTime=0.1}
    alphaScaleBiasClamp {bInterpResult=true}
    boneToModify {boneName=dyn_belt_snake__MSTR}
    rotation {pitch=10}
   }
   n19 = "Transform (Modify) Bone - Bone: dyn_thigh_armor"(ComponentPose=@n18, Alpha=Get ThighRotR_0() -> (ThighRotR_0:double), Translation="-5,0,1", Rotation="0,0,0", Scale="1,1,1") {
    rotationMode=BMM_Additive
    rotationSpace=BCS_BoneSpace
    translationMode=BMM_Additive
    translationSpace=BCS_BoneSpace
    alphaBoolBlend {blendOption=Linear}
    alphaScaleBiasClamp {bInterpResult=true}
    boneToModify {boneName=dyn_thigh_armor}
    translation {x=-5, z=1}
   }
   Save cached pose 'presim'(Pose=Component To Local(ComponentPose=@n19)) {cacheName=presim}
  ]}
 EventGraph {nodes [
   Event BlueprintUpdateAnimation()
   > n1 = Sequence!(execute) -> (then_0, then_1, then_2:Exec)
   > n2 = Branch(Condition=AND Boolean(A=Get bIsMoving2D(), B=Get bIsCrouching()))
   > Set Crouch Walk(Crouch Walk:double=1) -> (Output_Get:double)
   Set Crouch Walk(execute=@n2.else, Crouch Walk=0)
   n3 = Get Skel_Mesh()
   Set ThighRotL_0(
    execute=@n1.then_1
    ThighRotL_0:double=GetAlphaFromReferenceTransformDelta(SkeletalMeshComponent=@n3, BoneName=thigh_l, BaseName=pelvis, InputRange=(X=-20,Y=-100), TwistAxis="1,0,0")
   ) -> (Output_Get:double)
   > n4 = Set ThighRotR_0(ThighRotR_0:double=GetAlphaFromReferenceTransformDelta(SkeletalMeshComponent=@n3, BoneName=thigh_r, BaseName=pelvis, InputRange=(X=-20,Y=-100), TwistAxis="1,0,0")) -> (Output_Get:double)
   Set Frontend Pose(
    execute=[@n5.then, @n4.then]
    Frontend Pose:double=GetAlphaFromReferenceTransformDelta(SkeletalMeshComponent=@n3, BoneName=spine_01, BaseName=pelvis, DeltaComponentType=EulerX, InputRange=(X=10,Y=13), TwistAxis="1,0,0")
   ) -> (Output_Get:double)
   n5 = Branch(execute=@n1.then_2, Condition=Get bIsInFrontEnd())
   Set Frontend Pose(execute=@n5.else, Frontend Pose=0)
   Event BlueprintInitializeAnimation()
   > Set Skel_Mesh(Skel_Mesh=GetOwnerSkeletalMeshComponent())
   PlayerMovement(AnimInstance=Self-Reference(), XDotProductCoefficient=1, YDotProductCoefficient=1, ZDotProductCoefficient=1)
  ]}
]
variables [
 Crouch Walk {type=number}
 Frontend Pose {type=number}
 Skel_Mesh {type=SkeletalMeshComponent}
 ThighRotL_0 {type=number}
 ThighRotR_0 {type=number}
]
