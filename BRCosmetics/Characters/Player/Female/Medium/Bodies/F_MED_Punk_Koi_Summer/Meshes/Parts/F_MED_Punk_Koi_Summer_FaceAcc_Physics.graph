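# Physics asset for the F_MED_Punk_Koi_Summer face accessory mesh, covering its dynamic hair bones and the upper body around them.
# Capsule bodies cover upper spine, clavicles, neck and head, with an oversized second capsule on the right clavicle.
# Center, left and right three-bone hair chains hang from head-linked Lookat bodies, with tighter twist and side-chain swing limits.
# Angularly free constraints cross-link the center chain's lower two bones to both side chains.
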PhysicsAsset /BRCosmetics/Characters/Player/Female/Medium/Bodies/F_MED_Punk_Koi_Summer/Meshes/Parts/F_MED_Punk_Koi_Summer_FaceAcc_Physics
bodies {
 clavicle_l [
  {shapeType=Capsule, length=12.6717, radius=4.8754, shapeName=clavicle_l_capsule, center=<5 0 0>, rotation=<rot 90 0 0>}
  {shapeType=Capsule, length=4.503, radius=5.9933, shapeName=clavicle_l_capsule, center=<8.887 0 0.1779>, rotation=<rot 85 -0.001 -0.0011>}
 ]
 clavicle_r [
  {shapeType=Capsule, length=12.4134, radius=4.8754, shapeName=clavicle_r_capsule, center=<-5 0.1579 0>, rotation=<rot 90 0 0>}
  {shapeType=Capsule, length=14, radius=9.5, shapeName=clavicle_r_capsule, center=<-6.0388 -13.7408 -1.7319>, rotation=<rot 90 0 0>}
 ]
 dyn_hair_01 [{shapeType=Capsule, length=7.5287, radius=4.458, shapeName=dyn_hair_01_capsule, center=<5 -1.5 0>}]
 dyn_hair_02 [{shapeType=Capsule, length=7.7342, radius=4.7191, shapeName=dyn_hair_02_capsule, center=<4 -2.5 0>}]
 dyn_hair_03 [
  {shapeType=Capsule, length=11.1686, radius=3.8939, shapeName=dyn_hair_03_capsule, center=<5 -3.5 0>}
  {shapeType=Capsule, length=9, radius=2.5, shapeName=dyn_hair_03_capsule, center=<7.8882 -2.0094 0>}
 ]
 dyn_hair_L_01 [{shapeType=Capsule, length=2, radius=2, shapeName=dyn_hair_L_01_capsule, center=<3 0 0.5>, rotation=<rot 90 0 0>}]
 dyn_hair_L_02 [
  {shapeType=Capsule, length=3.6138, radius=2.6655, shapeName=dyn_hair_L_02_capsule, center=<4 0 1>, rotation=<rot 90 0 0>}
  {
   shapeType=Capsule
   length=3.196
   radius=1.4563
   shapeName=dyn_hair_L_02_capsule
   center=<3.967 3.5999 2.7195>
   rotation=<rot 84.7832 -165.6351 -160.8619>
  }
 ]
 dyn_hair_L_03 [
  {shapeType=Capsule, length=3.5845, radius=2.9155, shapeName=dyn_hair_L_03_capsule, center=<4 0 1>, rotation=<rot 90 -23.6294 -28.6295>}
  {
   shapeType=Capsule
   length=2.4557
   radius=1.5443
   shapeName=dyn_hair_L_03_capsule
   center=<8.7516 2.5088 0.0327>
   rotation=<rot -60 -159.999 0>
  }
 ]
 dyn_hair_L_Lookat [{shapeType=Capsule, length=0.505, radius=0.505, shapeName=dyn_hair_L_Lookat_capsule, rotation=<rot 0 -90 0>}]
 dyn_hair_Lookat [{shapeType=Capsule, length=0.505, radius=0.505, shapeName=dyn_hair_Lookat_capsule, rotation=<rot 0 -90 0>}]
 dyn_hair_R_01 [{shapeType=Capsule, length=2, radius=2, shapeName=dyn_hair_R_01_capsule, center=<3 0 0>, rotation=<rot 90 0 0>}]
 dyn_hair_R_02 [
  {shapeType=Capsule, length=4.75, radius=2.65, shapeName=dyn_hair_R_02_capsule, center=<4 0 -1.5>, rotation=<rot 90 0 0>}
  {shapeType=Capsule, length=3.25, radius=1, shapeName=dyn_hair_R_02_capsule, center=<4.4996 3.2209 -2.0612>, rotation=<rot 84.9888 0 0>}
 ]
 dyn_hair_R_03 [{shapeType=Capsule, length=4.25, radius=2.75, shapeName=dyn_hair_R_03_capsule, center=<4 0 -1.5>, rotation=<rot 90 0 0>}]
 dyn_hair_R_Lookat [{shapeType=Capsule, length=0.505, radius=0.505, shapeName=dyn_hair_R_Lookat_capsule, rotation=<rot 0 -90 0>}]
 head [{shapeType=Capsule, length=4.6867, radius=6.5305, shapeName=head_capsule, center=<-0.9804 2.5998 0>, rotation=<rot 90 90 77.5958>}]
 neck_01 [{shapeType=Capsule, length=3.255, radius=4.8, shapeName=neck_01_capsule, center=<-0.7844 1.2309 -0.198>, rotation=<rot 0 -89.9999 0>}]
 spine_04 [{shapeType=Capsule, length=7.2544, radius=10.4149, shapeName=spine_04_capsule, center=<2.5931 1.6603 0>, rotation=<rot 0 -92.6845 0>}]
 spine_05 [{shapeType=Capsule, length=17.5459, radius=7.5403, shapeName=spine_05_capsule, center=<1.675 -0.1132 -0.155>}]
}
constraints [
 {
  bone1Name=spine_05
  bone2Name=spine_04
  swing1LimitDegrees=45
  swing1Motion=Free
  swing2LimitDegrees=45
  swing2Motion=Free
  twistLimitDegrees=45
  twistMotion=Free
 }
 {
  bone1Name=clavicle_l
  bone2Name=spine_05
  swing1LimitDegrees=45
  swing1Motion=Free
  swing2LimitDegrees=45
  swing2Motion=Free
  twistLimitDegrees=45
  twistMotion=Free
 }
 {
  bone1Name=clavicle_r
  bone2Name=spine_05
  swing1LimitDegrees=45
  swing1Motion=Free
  swing2LimitDegrees=45
  swing2Motion=Free
  twistLimitDegrees=45
  twistMotion=Free
 }
 {
  bone1Name=head
  bone2Name=neck_01
  swing1LimitDegrees=45
  swing1Motion=Free
  swing2LimitDegrees=45
  swing2Motion=Free
  twistLimitDegrees=45
  twistMotion=Free
 }
 {
  bone1Name=neck_01
  bone2Name=spine_05
  swing1LimitDegrees=45
  swing1Motion=Free
  swing2LimitDegrees=45
  swing2Motion=Free
  twistLimitDegrees=45
  twistMotion=Free
 }
 {
  bone1Name=dyn_hair_Lookat
  bone2Name=head
  swing1LimitDegrees=45
  swing1Motion=Free
  swing2LimitDegrees=45
  swing2Motion=Free
  twistLimitDegrees=45
  twistMotion=Free
 }
 {bone1Name=dyn_hair_01, bone2Name=dyn_hair_Lookat, swing1LimitDegrees=35, swing2LimitDegrees=35, twistLimitDegrees=15}
 {bone1Name=dyn_hair_02, bone2Name=dyn_hair_01, swing1LimitDegrees=35, swing2LimitDegrees=35, twistLimitDegrees=15}
 {bone1Name=dyn_hair_03, bone2Name=dyn_hair_02, swing1LimitDegrees=35, swing2LimitDegrees=35, twistLimitDegrees=15}
 {bone1Name=dyn_hair_R_03, bone2Name=dyn_hair_R_02, swing1LimitDegrees=35, swing2LimitDegrees=15, twistLimitDegrees=15}
 {
  bone1Name=dyn_hair_R_Lookat
  bone2Name=head
  swing1LimitDegrees=45
  swing1Motion=Free
  swing2LimitDegrees=45
  swing2Motion=Free
  twistLimitDegrees=45
  twistMotion=Free
 }
 {bone1Name=dyn_hair_R_01, bone2Name=dyn_hair_R_Lookat, swing1LimitDegrees=35, swing2LimitDegrees=15, twistLimitDegrees=15}
 {bone1Name=dyn_hair_R_02, bone2Name=dyn_hair_R_01, swing1LimitDegrees=35, swing2LimitDegrees=15, twistLimitDegrees=15}
 {bone1Name=dyn_hair_L_03, bone2Name=dyn_hair_L_02, swing1LimitDegrees=35, swing2LimitDegrees=15, twistLimitDegrees=15}
 {
  bone1Name=dyn_hair_L_Lookat
  bone2Name=head
  swing1LimitDegrees=45
  swing1Motion=Free
  swing2LimitDegrees=45
  swing2Motion=Free
  twistLimitDegrees=45
  twistMotion=Free
 }
 {bone1Name=dyn_hair_L_01, bone2Name=dyn_hair_L_Lookat, swing1LimitDegrees=35, swing2LimitDegrees=15, twistLimitDegrees=15}
 {bone1Name=dyn_hair_L_02, bone2Name=dyn_hair_L_01, swing1LimitDegrees=35, swing2LimitDegrees=15, twistLimitDegrees=15}
 {
  bone1Name=dyn_hair_03
  bone2Name=dyn_hair_R_03
  swing1LimitDegrees=45
  swing1Motion=Free
  swing2LimitDegrees=45
  swing2Motion=Free
  twistLimitDegrees=45
  twistMotion=Free
 }
 {
  bone1Name=dyn_hair_02
  bone2Name=dyn_hair_R_02
  swing1LimitDegrees=45
  swing1Motion=Free
  swing2LimitDegrees=45
  swing2Motion=Free
  twistLimitDegrees=45
  twistMotion=Free
 }
 {
  bone1Name=dyn_hair_02
  bone2Name=dyn_hair_L_02
  swing1LimitDegrees=45
  swing1Motion=Free
  swing2LimitDegrees=45
  swing2Motion=Free
  twistLimitDegrees=45
  twistMotion=Free
 }
 {
  bone1Name=dyn_hair_03
  bone2Name=dyn_hair_L_03
  swing1LimitDegrees=45
  swing1Motion=Free
  swing2LimitDegrees=45
  swing2Motion=Free
  twistLimitDegrees=45
  twistMotion=Free
 }
]
