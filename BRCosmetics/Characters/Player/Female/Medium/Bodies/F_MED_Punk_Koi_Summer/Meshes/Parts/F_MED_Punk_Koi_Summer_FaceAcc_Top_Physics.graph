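# Physics asset for the top face-accessory part of the female Punk Koi Summer character cosmetic, defining bodies for its dynamic hair and ponytail.
# Capsules on upper spine, clavicles, neck and head form the body volume, joined by constraints with free rotation.
# Ponytail capsules chain from the head with a two-part side branch, and four extra hair capsules attach to the head, all with tight twist limits.
# Three flat hair boxes chain from a small Lookat box on the head with free rotation.
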PhysicsAsset /BRCosmetics/Characters/Player/Female/Medium/Bodies/F_MED_Punk_Koi_Summer/Meshes/Parts/F_MED_Punk_Koi_Summer_FaceAcc_Top_Physics
bodies {
 clavicle_l [{shapeType=Capsule, length=12.6717, radius=4.8754, shapeName=clavicle_l_capsule, center=<5 0 0>, rotation=<rot 90 0 0>}]
 clavicle_r [{shapeType=Capsule, length=13.1634, radius=4.8754, shapeName=clavicle_r_capsule, center=<-5 0 0>, rotation=<rot 90 0 0>}]
 dyn_hair_01 [{shapeType=Box, extentX=7.5, extentY=1, extentZ=10, shapeName=dyn_hair_01_box, center=<3.9702 4.321 0>, rotation=<rot 0 -8.3575 0>}]
 dyn_hair_02 [{shapeType=Box, extentX=10, extentY=0.5, extentZ=10, shapeName=dyn_hair_02_box, center=<4.5703 2.6462 0>, rotation=<rot 0 -15.3419 0>}]
 dyn_hair_03 [{shapeType=Box, extentX=10, extentY=0.5, extentZ=10, shapeName=dyn_hair_03_box, center=<4.2141 0.6019 0>, rotation=<rot 0 -19.5829 0>}]
 dyn_hair_Lookat [{shapeType=Box, extentX=0.505, extentY=0.505, extentZ=0.505, shapeName=dyn_hair_Lookat_box, center=<4 0 0>}]
 dyn_hair_extra_a_01 [{shapeType=Capsule, length=5, radius=1, shapeName=dyn_hair_extra_a_01_capsule, center=<4 0 0>, rotation=<rot 90 0 0>}]
 dyn_hair_extra_b_01 [{shapeType=Capsule, length=5, radius=1, shapeName=dyn_hair_extra_b_01_capsule, center=<4 0 0>, rotation=<rot 90 0 0>}]
 dyn_hair_extra_c_01 [{shapeType=Capsule, length=5, radius=1, shapeName=dyn_hair_extra_c_01_capsule, center=<4 0 0>, rotation=<rot 90 0 0>}]
 dyn_hair_extra_d_01 [{shapeType=Capsule, length=5, radius=1, shapeName=dyn_hair_extra_d_01_capsule, center=<4 0 0>, rotation=<rot 90 0 0>}]
 dyn_ponytail_01 [{shapeType=Capsule, length=6.1773, radius=5.1184, shapeName=dyn_ponytail_01_capsule, center=<4 1.3 0>}]
 dyn_ponytail_02 [{shapeType=Capsule, length=6.1652, radius=4.65, shapeName=dyn_ponytail_02_capsule, center=<4.0042 1.4 0>}]
 dyn_ponytail_extra_01 [{shapeType=Capsule, length=3.065, radius=1.9, shapeName=dyn_ponytail_extra_01_capsule, center=<3 0 0>, rotation=<rot 90 0 0>}]
 dyn_ponytail_extra_02 [{shapeType=Capsule, length=3.2368, radius=1.583, shapeName=dyn_ponytail_extra_02_capsule, center=<3 0 0>, rotation=<rot 90 0 0>}]
 head [{shapeType=Capsule, length=9.1867, radius=6.5305, shapeName=head_capsule, center=<-0.9804 2.5998 0>, rotation=<rot 90 90 72.5957>}]
 neck_01 [{shapeType=Capsule, length=3.255, radius=4.8, shapeName=neck_01_capsule, center=<-0.3384 1.0799 0>, rotation=<rot 0 -89.9999 0>}]
 spine_04 [{shapeType=Capsule, length=7.2544, radius=10.4149, shapeName=spine_04_capsule, center=<2.5931 1.6603 0>, rotation=<rot 0 -92.6845 0>}]
 spine_05 [{shapeType=Capsule, length=17.5459, radius=7.5403, shapeName=spine_05_capsule, center=<1.675 -0.1132 -0.155>}]
}
constraints [
 {
  bone1Name=spine_05
  bone2Name=spine_04
  swing1LimitDegrees=45
  swing1Motion=Free
  swing2LimitDegrees=45
  swing2Motion=Free
  twistLimitDegrees=45
  twistMotion=Free
 }
 {
  bone1Name=clavicle_l
  bone2Name=spine_05
  swing1LimitDegrees=45
  swing1Motion=Free
  swing2LimitDegrees=45
  swing2Motion=Free
  twistLimitDegrees=45
  twistMotion=Free
 }
 {
  bone1Name=clavicle_r
  bone2Name=spine_05
  swing1LimitDegrees=45
  swing1Motion=Free
  swing2LimitDegrees=45
  swing2Motion=Free
  twistLimitDegrees=45
  twistMotion=Free
 }
 {
  bone1Name=head
  bone2Name=neck_01
  swing1LimitDegrees=45
  swing1Motion=Free
  swing2LimitDegrees=45
  swing2Motion=Free
  twistLimitDegrees=45
  twistMotion=Free
 }
 {
  bone1Name=neck_01
  bone2Name=spine_05
  swing1LimitDegrees=45
  swing1Motion=Free
  swing2LimitDegrees=45
  swing2Motion=Free
  twistLimitDegrees=45
  twistMotion=Free
 }
 {bone1Name=dyn_ponytail_01, bone2Name=head, swing1LimitDegrees=35, swing2LimitDegrees=35, twistLimitDegrees=5}
 {bone1Name=dyn_ponytail_02, bone2Name=dyn_ponytail_01, swing1LimitDegrees=35, swing2LimitDegrees=35, twistLimitDegrees=5}
 {bone1Name=dyn_ponytail_extra_02, bone2Name=dyn_ponytail_extra_01, swing1LimitDegrees=35, swing2LimitDegrees=35, twistLimitDegrees=5}
 {bone1Name=dyn_ponytail_extra_01, bone2Name=dyn_ponytail_01, swing1LimitDegrees=35, swing2LimitDegrees=35, twistLimitDegrees=5}
 {bone1Name=dyn_hair_extra_a_01, bone2Name=head, swing1LimitDegrees=35, swing2LimitDegrees=35, twistLimitDegrees=3}
 {bone1Name=dyn_hair_extra_d_01, bone2Name=head, swing1LimitDegrees=35, swing2LimitDegrees=35, twistLimitDegrees=3}
 {bone1Name=dyn_hair_extra_c_01, bone2Name=head, swing1LimitDegrees=35, swing2LimitDegrees=35, twistLimitDegrees=3}
 {bone1Name=dyn_hair_extra_b_01, bone2Name=head, swing1LimitDegrees=35, swing2LimitDegrees=35, twistLimitDegrees=3}
 {
  bone1Name=dyn_hair_Lookat
  bone2Name=head
  swing1LimitDegrees=45
  swing1Motion=Free
  swing2LimitDegrees=45
  swing2Motion=Free
  twistLimitDegrees=45
  twistMotion=Free
 }
 {
  bone1Name=dyn_hair_01
  bone2Name=dyn_hair_Lookat
  swing1LimitDegrees=45
  swing1Motion=Free
  swing2LimitDegrees=45
  swing2Motion=Free
  twistLimitDegrees=45
  twistMotion=Free
 }
 {
  bone1Name=dyn_hair_02
  bone2Name=dyn_hair_01
  swing1LimitDegrees=45
  swing1Motion=Free
  swing2LimitDegrees=45
  swing2Motion=Free
  twistLimitDegrees=45
  twistMotion=Free
 }
 {
  bone1Name=dyn_hair_03
  bone2Name=dyn_hair_02
  swing1LimitDegrees=45
  swing1Motion=Free
  swing2LimitDegrees=45
  swing2Motion=Free
  twistLimitDegrees=45
  twistMotion=Free
 }
]
